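; Cosmetic Blueprint for the CraneLaugh pickaxe that drives its swing effects.
; On equip it caches the instigator pawn and resolves the active Particle cosmetic variant index.
; Right and left swings share identical logic, gated by CanExecuteCosmeticEvents, that checks the active montage section.
; During "Swing1" it spawns the thrust FX on the handle with a "Variant" parameter and hides the anim trail; other sections show the trail.
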(event EventBeginPlay)

(event FortWeapon|EventOnEquippedAndAttached
  ;[0.009047 0.4408 1 0.5 50] INIT FORT PLAYER PAWN
  (|CallParentFunction)
  (bind _asfort_player_pawn (Utilities|Casting|CastToFortPlayerPawn (Variables|Actor|GetInstigator)))
  (Variables|Default|SetAs%20Fort%20Player%20Pawn _asfort_player_pawn)
  ;; Variant Setup
  (bind _index (CosmeticLoadout|GetVariantChannelActiveIndexforSlotActor "(TagName=\"LoadoutSlot.Pickaxe\")" _asfort_player_pawn "(TagName=\"Cosmetics.Variant.Channel.Particle\")" (Utilities|Array|MakeArray "(TagName=\"Cosmetics.Variant.Property.Particle1\")" "(TagName=\"Cosmetics.Variant.Property.Particle2\")" "(TagName=\"Cosmetics.Variant.Property.Particle3\")" "(TagName=\"Cosmetics.Variant.Property.Particle4\")" "(TagName=\"Cosmetics.Variant.Property.Particle5\")" "(TagName=\"Cosmetics.Variant.Property.Particle6\")")))
  (Variables|Default|SetFE_VariantIndex _index))

(event EventTick
  (params (DeltaSeconds :type Float%20%28single-precision%29)))

(event EventSwingRight
  (bind _returnvalue (Game|GetInstigator))
  (bind _asfort_pawn (Utilities|Casting|CastToFortPawn _returnvalue))
  (bind _mesh (Variables|/Script/Engine.Character|GetMesh _asfort_pawn))
  (bind _returnvalue_2 (Components|SkeletalMesh|GetAnimInstance _mesh))
  (bind _returnvalue_3 (Animation|Montage|GetCurrentActiveMontage _returnvalue_2))
  (bind _returnvalue_4 (Animation|Montage|MontageGetCurrentSection _returnvalue_2 _returnvalue_3))
  (bind _returnvalue_5 (== _returnvalue_4 "Swing1"))
  ;[0.1105 0.04106 0.15 0.5 50] SWING
  (|CallParentFunction)
  (Networking|CanExecuteCosmeticEvents
    (:True
      (Utilities|FlowControl|Sequence
        (:then_0
          (if _returnvalue_5
            (bind _returnvalue_6 (Effects|SpawnFXSystemAttached (Variables|Default|GetFXThrust) (Variables|Visual|GetWeaponMesh) "Melee_TwoHanded_Handle"))
            (Effects|Components|ParticleSystem|SetFloatParameter _returnvalue_6 "Variant" (Math|Conversions|ToFloat(Integer) (Variables|Default|GetFE_VariantIndex)))))
        (:then_1
          (Development|SetHiddeninGame (CallFunction|GetAnimTrailVFXComponent) _returnvalue_5 false))))))

(event EventSwingLeft
  (bind _returnvalue (Game|GetInstigator))
  (bind _asfort_pawn (Utilities|Casting|CastToFortPawn _returnvalue))
  (bind _mesh (Variables|/Script/Engine.Character|GetMesh _asfort_pawn))
  (bind _returnvalue_2 (Components|SkeletalMesh|GetAnimInstance _mesh))
  (bind _returnvalue_3 (Animation|Montage|GetCurrentActiveMontage _returnvalue_2))
  (bind _returnvalue_4 (Animation|Montage|MontageGetCurrentSection _returnvalue_2 _returnvalue_3))
  (bind _returnvalue_5 (== _returnvalue_4 "Swing1"))
  ;[0.1105 0.04106 0.15 0.5 50] SWING
  (|CallParentFunction)
  (Networking|CanExecuteCosmeticEvents
    (:True
      (Utilities|FlowControl|Sequence
        (:then_0
          (if _returnvalue_5
            (bind _returnvalue_6 (Effects|SpawnFXSystemAttached (Variables|Default|GetFXThrust) (Variables|Visual|GetWeaponMesh) "Melee_TwoHanded_Handle"))
            (Effects|Components|ParticleSystem|SetFloatParameter _returnvalue_6 "Variant" (Math|Conversions|ToFloat(Integer) (Variables|Default|GetFE_VariantIndex)))))
        (:then_1
          (Development|SetHiddeninGame (CallFunction|GetAnimTrailVFXComponent) _returnvalue_5 false))))))

(fn ConstructionScript
  (|CallParentFunction))
