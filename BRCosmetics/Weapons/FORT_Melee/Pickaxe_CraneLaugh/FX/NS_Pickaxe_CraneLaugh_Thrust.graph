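# Niagara effect for the thrust of the CraneLaugh pickaxe, layering short, slightly delayed bursts positioned by sampling a skeletal mesh.
# Airflow spawns two mesh groups split by visibility tag, Head_Shockwave adds an expanding, fading mesh, and Sparks and Symbols launch dragged sprites.
# User.Variant selects among six palettes: variants 0–1 color the sparks, while 2–5 spawn curl-noise-driven symbols instead.
# A system time-of-day multiplier tints every layer except Airflow.
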NiagaraSystem /BRCosmetics/Weapons/FORT_Melee/Pickaxe_CraneLaugh/FX/NS_Pickaxe_CraneLaugh_Thrust
dependencies {
 usedDataInterfaces [
  &NiagaraDataInterfaceCurve
  &NiagaraDataInterfaceColorCurve
  &NiagaraDataInterfaceVector2DCurve
  &NiagaraDataInterfaceArrayFloat
  &NiagaraDataInterfaceSkeletalMesh
 ]
 usedDynamicInputs [&SetBoolByIntComparison]
 usedModules [
  &self:SystemSpawnScript.SystemScriptSource.SystemScriptGraph.NiagaraNodeAssignment_0.SetVariables_#1
  &self:SystemSpawnScript.SystemScriptSource.SystemScriptGraph.NiagaraNodeAssignment_2.SetVariables_#2
  &self:SystemSpawnScript.SystemScriptSource.SystemScriptGraph.NiagaraNodeAssignment_3.SetVariables_#3
  &self:SystemSpawnScript.SystemScriptSource.SystemScriptGraph.NiagaraNodeAssignment_4.SetVariables_#4
  &self:SystemSpawnScript.SystemScriptSource.SystemScriptGraph.NiagaraNodeAssignment_5.SetVariables_#5
  &self:SystemSpawnScript.SystemScriptSource.SystemScriptGraph.NiagaraNodeAssignment_6.SetVariables_#6
  &self:SystemSpawnScript.SystemScriptSource.SystemScriptGraph.NiagaraNodeAssignment_7.SetVariables_#7
  &SystemState
  &self:SystemSpawnScript.SystemScriptSource.SystemScriptGraph.NiagaraNodeAssignment_1.SetVariables_#8
 ]
}
emitters {
 Airflow {
  emitterUpdateScript [
   &EmitterState(Life Cycle Mode=::NewEnumerator1, Loop Behavior=::NewEnumerator1, Loop Delay=0.07, Loop Duration=0.5, UseLoopDelay=true)
   &SpawnBurst_Instantaneous(Use Spawn Probability=false, UseLoopCountLimit=false)
   &SpawnBurst_Instantaneous(Spawn Group=1, Use Spawn Probability=false, UseLoopCountLimit=false) {module=SpawnBurst_Instantaneous001}
  ]
  particleSpawnScript [
   /SetVariables(Particles.VisibilityTag=$Emitter.SpawnGroup) {isSetParameters=true}
   &InitializeParticle(Lifetime=0.25, Mesh Scale Mode=::NewEnumerator3, UsePositionOffset=false, Write Material Random=true)
   &AddVelocity(Velocity=&MergeHelper_VectorSwitch(0=<0 0 -100>, 1=<0 0 0>, 2=?, 3=?, Index=$Emitter.SpawnGroup, Num=::NewEnumerator0))
   &InitialMeshOrientation(Mesh Orientation Mode=::NewEnumerator2, Rotation=<0 -0.25 0>, Use Rotation Vector=true)
   &SkeletalMeshLocation(
    Bone Sampling Mode=::NewEnumerator4
    MaskSpawnGroup=false
    Orientation Sampling=::NewEnumerator1
    Sampled Position Offset=<0 -5 120>
    Skeletal Mesh=$System.Skeletal Mesh
   )
   &CameraOffset(Camera Offset Amount=30)
  ]
  particleUpdateScript [
   &ParticleState()
   &ScaleMeshSize(Scale Factor=&MakeVector(
     X=&MergeHelper_FloatSwitch(
      0=&FloatFromCurve(FloatCurve=DataInterface() {propertyValues {Curve {keys [<at 0 1>,<at 1 1.5>]}}})
      1=&FloatFromCurve(FloatCurve=DataInterface() {propertyValues {Curve {keys [<at 0 1>,<at 1 2.25>]}}})
      2=?
      3=?
      Index=$Emitter.SpawnGroup
      Num=::NewEnumerator0
     )
     Y=1
     Z=1
    ))
   &SolveForcesAndVelocity(Clamp Velocity=false, Limit Acceleration=false)
   &DynamicMaterialParameters(
    Index 0 Param 1=&FloatFromCurve(FloatCurve=DataInterface() {propertyValues {Curve {keys [<at 0 0>,<at 0.5 0>,<at 1 0.08>]}}})
    Index 0 Param 2=&FloatFromCurve(FloatCurve=DataInterface() {propertyValues {Curve {keys [{arriveTangent=-2.2616, interpMode=RCIM_Cubic, time=0, value=-0.5}, {interpMode=RCIM_Cubic, leaveTangent=-0.7156, time=1, value=1}]}}})
    Index 0 Param 3=1.2
    Index 0 Param 4=&FloatFromCurve(FloatCurve=DataInterface() {propertyValues {Curve {keys [<at 0 1>,<at 1 2>]}}}, Scale Curve=0.2)
    Module.Param0WriteEnabled=true
    Module.Param1WriteEnabled=true
    Module.Param2WriteEnabled=true
    Module.Param3WriteEnabled=true
   )
   &ScaleColor(
    Scale Alpha=&FloatFromCurve(FloatCurve=DataInterface() {propertyValues {Curve {keys [{tangentMode=RCTM_User, time=0, value=1}, {tangentMode=RCTM_User, time=1, value=0}]}}})
    Scale RGB=&MakeVectorFromLinearColorRGB(Color=&MergeHelper_ColorSwitch(
      0=&ColorFromCurve(ColorCurve=DataInterface() {propertyValues {
         AlphaCurve {keys []}
         BlueCurve {keys [<at 0.3237 2.8748 RCIM_Cubic>,<at 0.772 20 RCIM_Cubic>]}
         GreenCurve {keys [<at 0.3237 13.3186 RCIM_Cubic>,<at 0.772 12.235 RCIM_Cubic>]}
         RedCurve {keys [<at 0.3237 20 RCIM_Cubic>,<at 0.772 9.8343 RCIM_Cubic>]}
        }})
      1=&ColorFromCurve(ColorCurve=DataInterface() {propertyValues {
         AlphaCurve {keys []}
         BlueCurve {keys [<at 0.3237 20 RCIM_Cubic>,<at 0.772 20 RCIM_Cubic>]}
         GreenCurve {keys [<at 0.3237 12.7566 RCIM_Cubic>,<at 0.772 12.235 RCIM_Cubic>]}
         RedCurve {keys [<at 0.3237 8.083 RCIM_Cubic>,<at 0.772 9.8343 RCIM_Cubic>]}
        }})
      2=&ColorFromCurve(ColorCurve=DataInterface() {propertyValues {
         AlphaCurve {keys []}
         BlueCurve {keys [<at 0.3237 1.6182 RCIM_Cubic>,<at 0.772 9.8343 RCIM_Cubic>]}
         GreenCurve {keys [<at 0.3237 12.4363 RCIM_Cubic>,<at 0.772 14.9171 RCIM_Cubic>]}
         RedCurve {keys [<at 0.3237 20 RCIM_Cubic>,<at 0.772 20 RCIM_Cubic>]}
        }})
      3=&ColorFromCurve(ColorCurve=DataInterface() {propertyValues {
         AlphaCurve {keys []}
         BlueCurve {keys [<at 0.3237 20 RCIM_Cubic>,<at 0.772 20 RCIM_Cubic>]}
         GreenCurve {keys [<at 0.3237 12.7566 RCIM_Cubic>,<at 0.772 12.351 RCIM_Cubic>]}
         RedCurve {keys [<at 0.3237 8.083 RCIM_Cubic>,<at 0.772 19.9181 RCIM_Cubic>]}
        }})
      4=&ColorFromCurve(ColorCurve=DataInterface() {propertyValues {
         AlphaCurve {keys []}
         BlueCurve {keys [<at 0.3237 20 RCIM_Cubic>,<at 0.772 20 RCIM_Cubic>]}
         GreenCurve {keys [<at 0.3237 12.3674 RCIM_Cubic>,<at 0.772 15.4807 RCIM_Cubic>]}
         RedCurve {keys [<at 0.3237 0.9184 RCIM_Cubic>,<at 0.772 8.7019 RCIM_Cubic>]}
        }})
      5=&ColorFromCurve(ColorCurve=DataInterface() {propertyValues {
         AlphaCurve {keys []}
         BlueCurve {keys [<at 0.3237 20 RCIM_Cubic>,<at 0.772 20 RCIM_Cubic>]}
         GreenCurve {keys [<at 0.3237 11.7624 RCIM_Cubic>,<at 0.772 10.9253 RCIM_Cubic>]}
         RedCurve {keys [<at 0.3237 0.9925 RCIM_Cubic>,<at 0.772 3.9367 RCIM_Cubic>]}
        }})
      Index=&MakeIntFromFloat(Float=$User.Variant)
      Num=::NewEnumerator4
     ))
    ScaleA=true
    ScaleRGB=true
   )
   /SetVariables(Particles.DynamicMatDark:Vector4f=&MakeVector4(
     X=&FloatFromCurve(FloatCurve=DataInterface() {propertyValues {Curve {keys [<at 0 0.4>,<at 1 0.8>]}}})
     Y=&FloatFromCurve(FloatCurve=DataInterface() {propertyValues {Curve {keys [
          {arriveTangent=-2.2616, interpMode=RCIM_Cubic, leaveTangent=-3.554, tangentMode=RCTM_Break, time=0, value=1}
          {arriveTangent=-0.8527, interpMode=RCIM_Cubic, leaveTangent=-0.8527, tangentMode=RCTM_User, time=0.8, value=0}
         ]}}})
     Z=0.6
     W=0
    )) {isSetParameters=true}
   /SetVariables(Particles.Dark_Color:LinearColor=&MakeLinearColorFromVectorAndFloat(
     "Vector (RGB)"=&MakeVectorFromLinearColorRGB(Color=&MergeHelper_ColorSwitch(
       0=&ColorFromCurve(ColorCurve=DataInterface() {propertyValues {
          AlphaCurve {keys []}
          BlueCurve {keys [<at 0 0.0034 RCIM_Cubic>,<at 0.4196 0.0098>]}
          GreenCurve {keys [<at 0 0.0089 RCIM_Cubic>,<at 0.4196 0>]}
          RedCurve {keys [<at 0 0.0104 RCIM_Cubic>,<at 0.4196 0.1302>]}
         }})
       1=&ColorFromCurve(ColorCurve=DataInterface() {propertyValues {
          AlphaCurve {keys []}
          BlueCurve {keys [<at 0 0.0034 RCIM_Cubic>,<at 0.4345 0.1302>]}
          GreenCurve {keys [<at 0 0.0089 RCIM_Cubic>,<at 0.4345 0.0892>]}
          RedCurve {keys [<at 0 0.0104 RCIM_Cubic>,<at 0.4345 0.0546>]}
         }})
       2=&ColorFromCurve(ColorCurve=DataInterface() {propertyValues {
          AlphaCurve {keys []}
          BlueCurve {keys [<at 0 0.015 RCIM_Cubic>,<at 0.4345 0.1022>]}
          GreenCurve {keys [<at 0 0.0518 RCIM_Cubic>,<at 0.4345 0.2372>]}
          RedCurve {keys [<at 0 0.0729 RCIM_Cubic>,<at 0.4345 0.3698>]}
         }})
       3=&ColorFromCurve(ColorCurve=DataInterface() {propertyValues {
          AlphaCurve {keys []}
          BlueCurve {keys [<at 0 0.1719 RCIM_Cubic>,<at 0.4345 0.1302>]}
          GreenCurve {keys [<at 0 0.1474 RCIM_Cubic>,<at 0.4345 0.0723>]}
          RedCurve {keys [<at 0 0.1027 RCIM_Cubic>,<at 0.4345 0.117>]}
         }})
       4=&ColorFromCurve(ColorCurve=DataInterface() {propertyValues {
          AlphaCurve {keys []}
          BlueCurve {keys [<at 0 0.1719 RCIM_Cubic>,<at 0.4345 0.3229>]}
          GreenCurve {keys [<at 0 0.127 RCIM_Cubic>,<at 0.4345 0.2825>]}
          RedCurve {keys [<at 0 0.0372 RCIM_Cubic>,<at 0.4345 0.1611>]}
         }})
       5=&ColorFromCurve(ColorCurve=DataInterface() {propertyValues {
          AlphaCurve {keys []}
          BlueCurve {keys [<at 0 0.099 RCIM_Cubic>,<at 0.4345 0.1302>]}
          GreenCurve {keys [<at 0 0.0611 RCIM_Cubic>,<at 0.4345 0.084>]}
          RedCurve {keys [<at 0 0.0164 RCIM_Cubic>,<at 0.4345 0.039>]}
         }})
       Index=&MakeIntFromFloat(Float=$User.Variant)
       Num=::NewEnumerator4
      ))
     "Float (Alpha)"=&FloatFromCurve(
      FloatCurve=DataInterface() {propertyValues {Curve {keys [{arriveTangent=-9.8943, interpMode=RCIM_Cubic, leaveTangent=-9.8943, tangentMode=RCTM_User, time=0, value=4}, <at 1 0>]}}}
      Scale Curve=&SelectFloatFromArray(
       Direct Array Index=&MakeIntFromFloat(Float=$User.Variant)
       Float Selection Array=DataInterface() {propertyValues {FloatData [1,1,1,1,1,1]}}
      )
     )
    )) {isSetParameters=true}
  ]
  rendererClasses [&NiagaraMeshRendererProperties]
  renderers [
   {rendererClass=&NiagaraMeshRendererProperties, rendererIndex=0}
   {rendererClass=&NiagaraMeshRendererProperties, rendererIndex=1}
  ]
 }
 Head_Shockwave {
  emitterUpdateScript [
   &EmitterState(Life Cycle Mode=::NewEnumerator1, Loop Behavior=::NewEnumerator1, Loop Delay=0.08, Loop Duration=0.5, UseLoopDelay=true)
   &SpawnBurst_Instantaneous(Use Spawn Probability=false, UseLoopCountLimit=false)
  ]
  particleSpawnScript [
   &InitializeParticle(
    Lifetime=0.2
    Mesh Scale=&Multiply_VectorByFloat(Vector=<0.42 0.14 0.14>, Float=1)
    Mesh Scale Mode=::NewEnumerator4
    UsePositionOffset=false
    Write Material Random=true
   )
   &InitialMeshOrientation(Mesh Orientation Mode=::NewEnumerator2, Rotation=<0 -0.25 0>, Use Rotation Vector=true)
   &SkeletalMeshLocation(Bone Sampling Mode=::NewEnumerator4, MaskSpawnGroup=false, Sampled Position Offset=<0 -2 152>, Skeletal Mesh=$System.Skeletal Mesh)
   /SetVariables(Particles.Position2:NiagaraPosition=&AddVectorToPosition(Position=$Particles.Position, Vector=<0 0 -50>)) {isSetParameters=true}
  ]
  particleUpdateScript [
   &ParticleState()
   &ScaleMeshSize(Scale Factor=&MakeVector(
     X=&FloatFromCurve(FloatCurve=DataInterface() {propertyValues {Curve {keys [{arriveTangent=0.286, interpMode=RCIM_Cubic, leaveTangent=0.286, tangentMode=RCTM_User, time=0, value=1}, <at 1 1.2>]}}})
     Y=&FloatFromCurve(FloatCurve=DataInterface() {propertyValues {Curve {keys [{arriveTangent=3.1084, interpMode=RCIM_Cubic, leaveTangent=3.1084, tangentMode=RCTM_User, time=0, value=0.5}, <at 0.3 1>]}}})
     Z=&FloatFromCurve(FloatCurve=DataInterface() {propertyValues {Curve {keys [{arriveTangent=3.1084, interpMode=RCIM_Cubic, leaveTangent=3.1084, tangentMode=RCTM_User, time=0, value=0.5}, <at 0.3 1>]}}})
    ))
   &ScaleColor(
    Scale Alpha=&FloatFromCurve(FloatCurve=DataInterface() {propertyValues {Curve {keys [<at 0 1>,<at 1 0>]}}}, Scale Curve=1.5)
    Scale RGB=&Multiply_Vector(
     A=&MakeVectorFromLinearColorRGB(Color=&MergeHelper_ColorSwitch(
       0=&ColorFromCurve(ColorCurve=DataInterface() {propertyValues {
          AlphaCurve {keys []}
          BlueCurve {keys [<at 0 0.7854 RCIM_Cubic>,<at 0.4076 0.5 RCIM_Cubic>]}
          GreenCurve {keys [<at 0 3.5942 RCIM_Cubic>,<at 0.4076 0.3059 RCIM_Cubic>]}
          RedCurve {keys [<at 0 5 RCIM_Cubic>,<at 0.4076 0.2459 RCIM_Cubic>]}
         }})
       1=&ColorFromCurve(ColorCurve=DataInterface() {propertyValues {
          AlphaCurve {keys []}
          BlueCurve {keys [<at 0 3 RCIM_Cubic>,<at 0.4076 0.5 RCIM_Cubic>]}
          GreenCurve {keys [<at 0 2.2492 RCIM_Cubic>,<at 0.4076 0.2858 RCIM_Cubic>]}
          RedCurve {keys [<at 0 1.5154 RCIM_Cubic>,<at 0.4076 0.2379 RCIM_Cubic>]}
         }})
       2=&ColorFromCurve(ColorCurve=DataInterface() {propertyValues {
          AlphaCurve {keys []}
          BlueCurve {keys [<at 0 0.1437 RCIM_Cubic>,<at 0.4076 0.511 RCIM_Cubic>]}
          GreenCurve {keys [<at 0 2.0479 RCIM_Cubic>,<at 0.4076 0.8126 RCIM_Cubic>]}
          RedCurve {keys [<at 0 3 RCIM_Cubic>,<at 0.4076 1 RCIM_Cubic>]}
         }})
       3=&ColorFromCurve(ColorCurve=DataInterface() {propertyValues {
          AlphaCurve {keys []}
          BlueCurve {keys [<at 0 2.9698 RCIM_Cubic>,<at 0.4033 0.776 RCIM_Cubic>]}
          GreenCurve {keys [<at 0 1.7969 RCIM_Cubic>,<at 0.4033 0.7084 RCIM_Cubic>]}
          RedCurve {keys [<at 0 3 RCIM_Cubic>,<at 0.4033 0.4729 RCIM_Cubic>]}
         }})
       4=&ColorFromCurve(ColorCurve=DataInterface() {propertyValues {
          AlphaCurve {keys []}
          BlueCurve {keys [<at 0 3 RCIM_Cubic>,<at 0.4076 0.5 RCIM_Cubic>]}
          GreenCurve {keys [<at 0 1.893 RCIM_Cubic>,<at 0.4076 0.4542 RCIM_Cubic>]}
          RedCurve {keys [<at 0 0.2585 RCIM_Cubic>,<at 0.4076 0.4375 RCIM_Cubic>]}
         }})
       5=&ColorFromCurve(ColorCurve=DataInterface() {propertyValues {
          AlphaCurve {keys []}
          BlueCurve {keys [<at 0 4 RCIM_Cubic>,<at 0.4076 1 RCIM_Cubic>]}
          GreenCurve {keys [<at 0 2.3525 RCIM_Cubic>,<at 0.4076 0.5881 RCIM_Cubic>]}
          RedCurve {keys [<at 0 0.1985 RCIM_Cubic>,<at 0.4076 0.0496 RCIM_Cubic>]}
         }})
       Index=&MakeIntFromFloat(Float=$User.Variant)
       Num=::NewEnumerator4
      ))
     B=&MakeVectorFromLinearColorRGB(Color=$System.TODColorMultiplier)
    )
    ScaleA=true
    ScaleRGB=true
   )
   &FN_Mesh_Attributes(
    Color=&MakeLinearColorFromVectorAndFloat(
     "Vector (RGB)"=&Multiply_Vector(
      A=&MakeVectorFromLinearColorRGB(Color=&MergeHelper_ColorSwitch(
        0=&ColorFromCurve(ColorCurve=DataInterface() {propertyValues {
           AlphaCurve {keys []}
           BlueCurve {keys [<at 0 0.3208 RCIM_Cubic>,<at 0.4266 1 RCIM_Cubic>]}
           GreenCurve {keys [<at 0 1.3686 RCIM_Cubic>,<at 0.4266 1 RCIM_Cubic>]}
           RedCurve {keys [<at 0 2 RCIM_Cubic>,<at 0.4266 1 RCIM_Cubic>]}
          }})
        1=&ColorFromCurve(ColorCurve=DataInterface() {propertyValues {
           AlphaCurve {keys []}
           BlueCurve {keys [<at 0 2 RCIM_Cubic>,<at 0.4266 1 RCIM_Cubic>]}
           GreenCurve {keys [<at 0 1.4995 RCIM_Cubic>,<at 0.4266 0.5716 RCIM_Cubic>]}
           RedCurve {keys [<at 0 1.0103 RCIM_Cubic>,<at 0.4266 0.4757 RCIM_Cubic>]}
          }})
        2=&ColorFromCurve(ColorCurve=DataInterface() {propertyValues {
           AlphaCurve {keys []}
           BlueCurve {keys [<at 0 0.2015 RCIM_Cubic>,<at 0.4266 0.848 RCIM_Cubic>]}
           GreenCurve {keys [<at 0 2.8407 RCIM_Cubic>,<at 0.4266 1.5392 RCIM_Cubic>]}
           RedCurve {keys [<at 0 5 RCIM_Cubic>,<at 0.4266 2 RCIM_Cubic>]}
          }})
        3=&ColorFromCurve(ColorCurve=DataInterface() {propertyValues {
           AlphaCurve {keys []}
           BlueCurve {keys [<at 0 5 RCIM_Cubic>,<at 0.307 0.8542 RCIM_Cubic>]}
           GreenCurve {keys [<at 0 2.5103 RCIM_Cubic>,<at 0.307 0.7149 RCIM_Cubic>]}
           RedCurve {keys [<at 0 4.41 RCIM_Cubic>,<at 0.307 0.4708 RCIM_Cubic>]}
          }})
        4=&ColorFromCurve(ColorCurve=DataInterface() {propertyValues {
           AlphaCurve {keys []}
           BlueCurve {keys [<at 0 5 RCIM_Cubic>,<at 0.4266 1 RCIM_Cubic>]}
           GreenCurve {keys [<at 0 3.155 RCIM_Cubic>,<at 0.4266 0.9 RCIM_Cubic>]}
           RedCurve {keys [<at 0 0.4309 RCIM_Cubic>,<at 0.4266 0.7524 RCIM_Cubic>]}
          }})
        5=&ColorFromCurve(ColorCurve=DataInterface() {propertyValues {
           AlphaCurve {keys []}
           BlueCurve {keys [<at 0 4 RCIM_Cubic>,<at 0.4266 2 RCIM_Cubic>]}
           GreenCurve {keys [<at 0 2.3525 RCIM_Cubic>,<at 0.4266 1.67 RCIM_Cubic>]}
           RedCurve {keys [<at 0 0.1985 RCIM_Cubic>,<at 0.4266 1.2546 RCIM_Cubic>]}
          }})
        Index=&MakeIntFromFloat(Float=$User.Variant)
        Num=::NewEnumerator4
       ))
      B=&MakeVectorFromLinearColorRGB(Color=$System.TODColorMultiplier)
     )
     "Float (Alpha)"=&FloatFromCurve(
      FloatCurve=DataInterface() {propertyValues {Curve {keys [{arriveTangent=-3.009, interpMode=RCIM_Cubic, leaveTangent=-3.009, tangentMode=RCTM_User, time=0, value=1}, <at 1 0>]}}}
      Scale Curve=&Multiply_Float(A=2, B=1)
     )
    )
    Create Color Attribute=true
    Create Scale Attribute=true
    Scale=&Multiply_Vector(
     A=&RandomRangeVector(Minimum=<0.12 0.25 0.25>, Maximum=<0.14 0.25 0.25>)
     B=&MakeVector(
      X=&FloatFromCurve(FloatCurve=DataInterface() {propertyValues {Curve {keys [{arriveTangent=0.3763, interpMode=RCIM_Cubic, leaveTangent=0.3763, tangentMode=RCTM_User, time=0, value=1}, <at 1 1.2>]}}})
      Y=&FloatFromCurve(FloatCurve=DataInterface() {propertyValues {Curve {keys [{arriveTangent=3.1084, interpMode=RCIM_Cubic, leaveTangent=3.1084, tangentMode=RCTM_User, time=0, value=0.5}, <at 0.3 1>]}}})
      Z=&FloatFromCurve(FloatCurve=DataInterface() {propertyValues {Curve {keys [{arriveTangent=3.1084, interpMode=RCIM_Cubic, leaveTangent=3.1084, tangentMode=RCTM_User, time=0, value=0.5}, <at 0.3 1>]}}})
     )
    )
    Write to Module Namespace=true
   ) {module=FN_Mesh_Attributes001}
   /SetVariables(Particles.Dy3:Vector4f=&MakeVector4(X=0, Y=1, Z=0, W=1)) {isSetParameters=true}
  ]
  rendererClasses [&NiagaraMeshRendererProperties]
  renderers [
   {rendererClass=&NiagaraMeshRendererProperties, rendererIndex=0}
   {rendererClass=&NiagaraMeshRendererProperties, rendererIndex=1}
  ]
 }
 Sparks {
  emitterUpdateScript [
   &EmitterState(Life Cycle Mode=::NewEnumerator1, Loop Behavior=::NewEnumerator1, Loop Delay=0.05, Loop Duration=0.5, UseLoopDelay=true)
   &SpawnBurst_Instantaneous(Spawn Count=10, Use Spawn Probability=false, UseLoopCountLimit=false)
  ]
  particleSpawnScript [
   &InitializeParticle(
    Color=&MultiplyLinearColor(
     Linear Color=&MergeHelper_ColorSwitch(
      0=<rgba 1.5 1.0785 0.3214 1>
      1=<rgba 0.3214 0.9303 1.5 1>
      2=<rgba 0 0 0 0>
      3=<rgba 0 0 0 0>
      4=<rgba 0 0 0 0>
      5=<rgba 0 0 0 0>
      Index=&MakeIntFromFloat(Float=$User.Variant)
      Num=::NewEnumerator4
     )
     Scale Factor=&MakeVector4_FromLinearColor(LinearColor=$System.TODColorMultiplier)
    )
    Lifetime Max=0.4
    Lifetime Min=0.3
    Lifetime Mode=::NewEnumerator1
    Sprite Size Max=<2 8>
    Sprite Size Min=<2 8>
    Sprite Size Mode=::NewEnumerator2
    UsePositionOffset=false
    Write Material Random=true
   )
   &SkeletalMeshLocation(Bone Sampling Mode=::NewEnumerator4, MaskSpawnGroup=false, Sampled Position Offset=<0 -3 -20>, Skeletal Mesh=$System.Skeletal Mesh)
   &AddVelocity(Velocity=<0 0 500>)
   &BoxLocation(Box Size=<20 20 100>, MaskSpawnGroup=false, Module.Override Seed=false, Rotate Box=false, Surface Only=false)
  ]
  particleUpdateScript [
   &ParticleState()
   /SetVariables(Particles.Glow03_Color=&MakeLinearColorFromVectorAndFloat(
     "Vector (RGB)"=&Multiply_Vector(
      A=&MakeVectorFromLinearColorRGB(Color=&MultiplyLinearColor(
        Linear Color=&MergeHelper_ColorSwitch(
         0=<rgba 1 0.9311 0.3291 0.2>
         1=<rgba 0.3291 0.7205 1 0.2>
         2=<rgba 0 0 0 0>
         3=<rgba 0 0 0 0>
         4=<rgba 0 0 0 0>
         5=<rgba 0 0 0 0>
         Index=&MakeIntFromFloat(Float=$User.Variant)
         Num=::NewEnumerator4
        )
        Scale Factor=&MakeVector4_FromFloat(Float=0.5)
       ))
      B=&MakeVectorFromLinearColorRGB(Color=$System.TODColorMultiplier)
     )
     "Float (Alpha)"=&Multiply_Float(A=0.2, B=1)
    )) {isSetParameters=true}
   /SetVariables(Particles.Glow03_Size=&Multiply_Vector2D_ByFloat(
     Float=&FloatFromCurve(
      FloatCurve=DataInterface() {propertyValues {Curve {keys [
          <at 0 0 RCIM_Cubic>
          <at 0.2 1.3 RCIM_Cubic>
          {arriveTangent=-0.1423, interpMode=RCIM_Cubic, leaveTangent=-0.1423, tangentMode=RCTM_User, time=1, value=0}
         ]}}}
      Scale Curve=15
     )
     Vector2D=<0.7 2>
    )) {isSetParameters=true}
   &ScaleColor(
    Scale Alpha=&FloatFromCurve(FloatCurve=DataInterface() {propertyValues {Curve {keys [<at 0.2 0>,<at 0.264 0>,<at 0.3 1>,<at 1 1>]}}})
    ScaleA=true
    ScaleRGB=true
   )
   &ScaleSpriteSize(Scale Factor=&Vector2DFromCurve(Vector2Curve=DataInterface() {propertyValues {XCurve {keys [<at 0 1>,<at 1 0>]}, YCurve {keys [<at 0 1>,<at 1 0>]}}}, ScaleCurve=<0.3 1>)) {module=ScaleSpriteSize001}
   &Drag(
    Drag=&FloatFromCurve(FloatCurve=DataInterface() {propertyValues {Curve {keys [
         {interpMode=RCIM_Cubic, leaveTangent=12.2797, tangentMode=RCTM_Break, time=0, value=1}
         {interpMode=RCIM_Cubic, tangentMode=RCTM_User, time=1, value=5}
        ]}}})
    Use Linear Drag=true
    Use Rotational Drag=false
   )
   &SolveForcesAndVelocity(Clamp Velocity=false, Limit Acceleration=false)
   &ScaleSpriteSizeBySpeed(Max Scale Factor=<1 2>)
  ]
  rendererClasses [&NiagaraSpriteRendererProperties]
  renderers [
   {rendererClass=&NiagaraSpriteRendererProperties}
   {rendererClass=&NiagaraSpriteRendererProperties, rendererIndex=1}
  ]
 }
 Symbols {
  emitterUpdateScript [
   &EmitterState(Life Cycle Mode=::NewEnumerator1, Loop Behavior=::NewEnumerator1, Loop Delay=0.05, Loop Duration=0.5, UseLoopDelay=true)
   &SpawnBurst_Instantaneous(
    Spawn Count=&MakeIntFromFloat(Float=&Multiply_Float(
      A=&SelectFloatFromArray(
       Direct Array Index=&MakeIntFromFloat(Float=$User.Variant)
       Float Selection Array=DataInterface() {propertyValues {FloatData [0,0,8,9,6,5]}}
      )
      B=&RandomRangeFloat(Minimum=1, Maximum=1.5)
     ))
    Use Spawn Probability=false
    UseLoopCountLimit=false
   )
  ]
  particleSpawnScript [
   &InitializeParticle(
    Color=&MergeHelper_ColorSwitch(
     0=<rgba 0 0 0 0>
     1=<rgba 0 0 0 0>
     2=&ColorFromCurve(
      ColorCurve=DataInterface() {propertyValues {
        AlphaCurve {keys [<at 0 1 RCIM_Cubic>]}
        BlueCurve {keys [<at 0.4274 0 RCIM_Cubic>,<at 0.4699 0 RCIM_Cubic>]}
        GreenCurve {keys [<at 0.4274 0.6232 RCIM_Cubic>,<at 0.4699 0.3648 RCIM_Cubic>]}
        RedCurve {keys [<at 0.4274 1 RCIM_Cubic>,<at 0.4699 1 RCIM_Cubic>]}
       }}
      CurveIndex=&RandomRangeFloat(Minimum=0, Maximum=1)
     )
     3=&ColorFromCurve(
      ColorCurve=DataInterface() {propertyValues {
        AlphaCurve {keys [<at 0 1 RCIM_Cubic>]}
        BlueCurve {keys [<at 0.4274 1 RCIM_Cubic>,<at 0.4699 1 RCIM_Cubic>]}
        GreenCurve {keys [<at 0.4274 0.7649 RCIM_Cubic>,<at 0.4699 0.3664 RCIM_Cubic>]}
        RedCurve {keys [<at 0.4274 0.3886 RCIM_Cubic>,<at 0.4699 0.7831 RCIM_Cubic>]}
       }}
      CurveIndex=&RandomRangeFloat(Minimum=0, Maximum=1)
     )
     4=&ColorFromCurve(
      ColorCurve=DataInterface() {propertyValues {
        AlphaCurve {keys [<at 0 1 RCIM_Cubic>]}
        BlueCurve {keys [<at 0.4274 1 RCIM_Cubic>,<at 0.4699 1 RCIM_Cubic>]}
        GreenCurve {keys [<at 0.4274 0.9469 RCIM_Cubic>,<at 0.4699 0.6167 RCIM_Cubic>]}
        RedCurve {keys [<at 0.4274 0.5208 RCIM_Cubic>,<at 0.4699 0.4375 RCIM_Cubic>]}
       }}
      CurveIndex=&RandomRangeFloat(Minimum=0, Maximum=1)
     )
     5=&ColorFromCurve(
      ColorCurve=DataInterface() {propertyValues {
        AlphaCurve {keys [<at 0 1 RCIM_Cubic>]}
        BlueCurve {keys [<at 0.4274 1 RCIM_Cubic>,<at 0.4699 1 RCIM_Cubic>]}
        GreenCurve {keys [<at 0.4274 0 RCIM_Cubic>,<at 0.4699 0.282 RCIM_Cubic>]}
        RedCurve {keys [<at 0.4274 0.6698 RCIM_Cubic>,<at 0.4699 0 RCIM_Cubic>]}
       }}
      CurveIndex=&RandomRangeFloat(Minimum=0, Maximum=1)
     )
     Index=&MakeIntFromFloat(Float=$User.Variant, Conversion Mode=::NewEnumerator1)
     Num=::NewEnumerator4
    )
    Lifetime=&MergeHelper_FloatSwitch(
     0=0
     1=0
     2=&RandomRangeFloat(Minimum=0.3, Maximum=0.4)
     3=&RandomRangeFloat(Minimum=0.3, Maximum=0.4)
     4=&RandomRangeFloat(Minimum=0.3, Maximum=0.4)
     5=&RandomRangeFloat(Minimum=0.3, Maximum=0.4)
     Index=&MakeIntFromFloat(Float=$User.Variant, Conversion Mode=::NewEnumerator1)
     Num=::NewEnumerator4
    )
    Ribbon Width=1
    Ribbon Width Mode=::NewEnumerator1
    Sprite Rotation Angle=&MergeHelper_FloatSwitch(
     0=0
     1=0
     2=0
     3=&RandomRangeFloat(Minimum=0, Maximum=1)
     4=0
     5=0
     Index=&MakeIntFromFloat(Float=$User.Variant, Conversion Mode=::NewEnumerator1)
     Num=::NewEnumerator4
    )
    Sprite Rotation Mode=::NewEnumerator1
    Sprite Size Mode=::NewEnumerator3
    Uniform Sprite Size=&MergeHelper_FloatSwitch(
     0=0
     1=0
     2=&RandomRangeFloat(Minimum=10, Maximum=14)
     3=&RandomRangeFloat(Minimum=3, Maximum=7)
     4=&RandomRangeFloat(Minimum=7, Maximum=10)
     5=&RandomRangeFloat(Minimum=15, Maximum=20)
     Index=&MakeIntFromFloat(Float=$User.Variant, Conversion Mode=::NewEnumerator1)
     Num=::NewEnumerator4
    )
    UsePositionOffset=true
    Write Material Random=true
   )
   &SkeletalMeshLocation(Bone Sampling Mode=::NewEnumerator4, MaskSpawnGroup=false, Sampled Position Offset=<0 -3 -20>, Skeletal Mesh=$System.Skeletal Mesh)
   &AddVelocity(Velocity=<0 0 500>)
   &BoxLocation(Box Size=<20 20 100>, MaskSpawnGroup=false, Module.Override Seed=false, Rotate Box=false, Surface Only=false)
   /SetVariables(Particles.SubImageIndexAlt:NiagaraFloat=&RandomRangeFloat(Minimum=0, Maximum=36)) {isSetParameters=true}
  ]
  particleUpdateScript [
   &ParticleState()
   &ScaleSpriteSize(Uniform Curve Sprite Scale=DataInterface() {propertyValues {Curve {keys [
        {arriveTangent=8.2127, interpMode=RCIM_Cubic, leaveTangent=8.2127, tangentMode=RCTM_User, time=0, value=0}
        {
         arriveTangent=0.0908
         arriveTangentWeight=0.232
         interpMode=RCIM_Cubic
         leaveTangent=0.0908
         leaveTangentWeight=2.2538
         tangentMode=RCTM_User
         time=0.1
         value=1
        }
        {arriveTangent=-0.042, interpMode=RCIM_Cubic, leaveTangent=-0.042, tangentMode=RCTM_User, time=1, value=0}
       ]}}}) {module=ScaleSpriteSize001}
   &Drag(
    Drag=&FloatFromCurve(FloatCurve=DataInterface() {propertyValues {Curve {keys [
         {interpMode=RCIM_Cubic, leaveTangent=12.2797, tangentMode=RCTM_Break, time=0, value=1}
         {interpMode=RCIM_Cubic, tangentMode=RCTM_User, time=1, value=5}
        ]}}})
    Use Linear Drag=true
    Use Rotational Drag=false
   ) {module=Drag001}
   &CurlNoiseForce(
    Noise Strength=&FloatFromCurve(
     FloatCurve=DataInterface() {propertyValues {Curve {keys [
         {arriveTangent=2.7279, interpMode=RCIM_Cubic, leaveTangent=2.7279, tangentMode=RCTM_User, time=0, value=0}
         {interpMode=RCIM_Cubic, tangentMode=RCTM_User, time=1, value=1}
        ]}}}
     Scale Curve=550
    )
    Pan Noise=false
   )
   &SolveForcesAndVelocity(Clamp Velocity=false, Limit Acceleration=false)
   &ScaleColor(
    Scale Alpha=&MakeFloatFromLinearColor(LinearColor=$System.TODColorMultiplier, Channel:ENiagaraLinearColor_Channels=::NewEnumerator3)
    Scale RGB=&MakeVectorFromLinearColorRGB(Color=$System.TODColorMultiplier)
    ScaleA=true
    ScaleRGB=true
   )
  ]
  rendererClasses [&NiagaraSpriteRendererProperties]
  renderers [
   {rendererClass=&NiagaraSpriteRendererProperties}
   {rendererClass=&NiagaraSpriteRendererProperties, rendererIndex=1}
   {rendererClass=&NiagaraSpriteRendererProperties, rendererIndex=2}
   {rendererClass=&NiagaraSpriteRendererProperties, rendererIndex=3}
  ]
 }
}
summary {
 systemName=NS_Pickaxe_CraneLaugh_Thrust
 emitters [
  {emitterName=Head_Shockwave, rendererClasses [&NiagaraMeshRendererProperties]}
  {emitterName=Sparks, rendererClasses [&NiagaraSpriteRendererProperties]}
  {emitterName=Airflow, rendererClasses [&NiagaraMeshRendererProperties]}
  {emitterName=Symbols, rendererClasses [&NiagaraSpriteRendererProperties]}
 ]
 userVariables [
  User.bIsFrontEnd {type=NiagaraBool}
  User.bIsFrontEndPreview {type=NiagaraBool}
  User.Debug {type=NiagaraBool, defaultValue=true}
  User.Variant {type=NiagaraFloat, defaultValue=0, description=""}
 ]
}
