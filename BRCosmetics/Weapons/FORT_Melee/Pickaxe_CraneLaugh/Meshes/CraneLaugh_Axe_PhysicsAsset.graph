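# Physics asset for the CraneLaugh pickaxe, a Fortnite cosmetic melee weapon, providing collision bodies and joints for its dangling chain.
# One long capsule covers the two-handed handle, a small sphere marks the chain root, and four identical capsules form the links.
# Constraints join the root to the links in series with twist locked, leaving bending to the swing axes.
# No constraint connects the handle body to the chain.
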PhysicsAsset /BRCosmetics/Weapons/FORT_Melee/Pickaxe_CraneLaugh/Meshes/CraneLaugh_Axe_PhysicsAsset
bodies {
 Melee_TwoHanded_Handle [{shapeType=Capsule, length=81.4914, radius=4.6273, center=<0.4104 -0.0477 3.7723>, rotation=<rot -0.1479 -90 0.0719>}]
 dyn_chain_01 [{shapeType=Capsule, length=2.5, radius=1, shapeName=dyn_chain_09_capsule, center=<2.3373 0 0>, rotation=<rot -90 90 -89.9999>}]
 dyn_chain_02 [{shapeType=Capsule, length=2.5, radius=1, shapeName=dyn_chain_09_capsule, center=<2.3373 0 0>, rotation=<rot -90 90 -89.9999>}]
 dyn_chain_03 [{shapeType=Capsule, length=2.5, radius=1, shapeName=dyn_chain_09_capsule, center=<2.3373 0 0>, rotation=<rot -90 90 -89.9999>}]
 dyn_chain_04 [{shapeType=Capsule, length=2.5, radius=1, shapeName=dyn_chain_09_capsule, center=<2.3373 0 0>, rotation=<rot -90 90 -89.9999>}]
 mstr_chain [{shapeType=Sphere, radius=0.505, shapeName=mast_chain_sphere}]
}
constraints [
 {bone1Name=dyn_chain_04, bone2Name=dyn_chain_03, swing1LimitDegrees=30, swing2LimitDegrees=30, twistLimitDegrees=45, twistMotion=Locked}
 {bone1Name=dyn_chain_03, bone2Name=dyn_chain_02, swing1LimitDegrees=30, swing2LimitDegrees=30, twistLimitDegrees=45, twistMotion=Locked}
 {bone1Name=dyn_chain_02, bone2Name=dyn_chain_01, swing1LimitDegrees=30, swing2LimitDegrees=30, twistLimitDegrees=45, twistMotion=Locked}
 {bone1Name=mstr_chain, bone2Name=dyn_chain_01, swing1LimitDegrees=30, swing2LimitDegrees=30, twistLimitDegrees=45, twistMotion=Locked}
]
